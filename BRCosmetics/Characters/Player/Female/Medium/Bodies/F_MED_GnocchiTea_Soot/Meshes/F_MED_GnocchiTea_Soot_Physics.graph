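# Physics asset for the F_MED_GnocchiTea_Soot female character body cosmetic, set up for a jacket with one dynamic bone per side.
# Each dyn_jacket01 box is constrained to a small sphere on its MSTR_jacket bone, with a wide first swing, very narrow second swing and limited twist.
# Capsules on the pelvis, three spine bones and both thighs form unconstrained body volumes around the jacket.
# Left and right shapes mirror each other.
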PhysicsAsset /BRCosmetics/Characters/Player/Female/Medium/Bodies/F_MED_GnocchiTea_Soot/Meshes/F_MED_GnocchiTea_Soot_Physics
bodies {
 MSTR_jacket_l [{shapeType=Sphere, radius=0.505, shapeName=MSTR_jacket_l_sphere}]
 MSTR_jacket_r [{shapeType=Sphere, radius=0.505, shapeName=MSTR_jacket_l_sphere}]
 dyn_jacket01_l [{
   shapeType=Box
   extentX=1.58
   extentY=8.1092
   extentZ=6.9939
   shapeName=dyn_jacket01_l_box
   center=<5.7643 0.175 3.1045>
   rotation=<rot 9.6864 -85.5885 0.2813>
  }]
 dyn_jacket01_r [{
   shapeType=Box
   extentX=1.58
   extentY=8.1092
   extentZ=6.9939
   shapeName=dyn_jacket01_l_box
   center=<5.7643 0.175 -3.1045>
   rotation=<rot -9.6864 -85.5884 -179.7187>
  }]
 pelvis [{shapeType=Capsule, length=10.978, radius=9.6191, shapeName=pelvis_capsule, center=<-2.0015 -1.8267 0>, rotation=<rot 0 -90 0>}]
 spine_01 [{shapeType=Capsule, length=7.6527, radius=8.1062, shapeName=spine_01_capsule, center=<-2.0543 0.3936 0>, rotation=<rot 0 -90 0>}]
 spine_02 [{shapeType=Capsule, length=4.9412, radius=7.0423, shapeName=spine_02_capsule, center=<0.006 2.4286 0>, rotation=<rot 0 -83.9837 0>}]
 spine_03 [{shapeType=Capsule, length=3.0429, radius=8.3381, shapeName=spine_03_capsule, center=<0 2.8597 0>, rotation=<rot 0 -90 0>}]
 thigh_l [{
   shapeType=Capsule
   length=18.8846
   radius=8.5736
   shapeName=thigh_l_capsule
   center=<-13.38 0 0.1263>
   rotation=<rot -89.8854 -0.0019 0.0016>
  }]
 thigh_r [{
   shapeType=Capsule
   length=18.8846
   radius=8.5736
   shapeName=thigh_l_capsule
   center=<13.38 0 -0.1263>
   rotation=<rot -89.8854 -0.0019 -179.9984>
  }]
}
constraints [
 {bone1Name=dyn_jacket01_l, bone2Name=MSTR_jacket_l, swing1LimitDegrees=45, swing2LimitDegrees=2, twistLimitDegrees=10}
 {bone1Name=dyn_jacket01_r, bone2Name=MSTR_jacket_r, swing1LimitDegrees=45, swing2LimitDegrees=2, twistLimitDegrees=10}
]
