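# Animation Blueprint for the F_MED_GnocchiTea_Soot cosmetic body, layering corrective and secondary motion onto the attached parent mesh's pose.
# The copied pose runs through a Control Rig, then a weighted constraint and two half-alpha Look Ats position the belt pin corrective bone between neighboring bones.
# A RigidBody node then simulates physics relative to spine_03 with world gravity overridden.
# Each update stores a Gravity Override vector computed by a deprecated state function.
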AnimBlueprint /BRCosmetics/Characters/Player/Female/Medium/Bodies/F_MED_GnocchiTea_Soot/Meshes/F_MED_GnocchiTea_Soot_AnimBP
target_skeleton=&/F_MED_GnocchiTea_Soot_Skeleton
graphs [
 AnimGraph {nodes [
   n1 = Control Rig(Source=Copy Pose From Mesh() {bUseAttachedParent=true}) {
    alphaBoolBlend {blendOption=Linear}
    controlRigAssetReference {blueprintRigClass=&/F_MED_GnocchiTea_Soot_CtrlRig.F_MED_GnocchiTea_Soot_CtrlRig_C}
   }
   n2 = "Constraint - def_beltPinCorr (PPPPPPPP)"!(
    ComponentPose=Local To Component(LocalPose=@n1)
    Alpha=1
    ConstraintWeights_0=0.0159
    ConstraintWeights_1:float=0.5695
    ConstraintWeights_2:float=0.0365
    ConstraintWeights_3:float=0.0376
    ConstraintWeights_4:float=0.1605
    ConstraintWeights_5:float=0.1662
    ConstraintWeights_6:float=0.0073
    ConstraintWeights_7:float=0.0065
   ) -> (Pose) {
    alphaBoolBlend {blendOption=Linear}
    boneToModify {boneName=def_beltPinCorr}
    constraintSetup [
     {targetBone {boneName=spine_02}}
     {targetBone {boneName=pelvis}}
     {targetBone {boneName=thigh_twist_01_l}}
     {targetBone {boneName=thigh_twist_01_r}}
     {targetBone {boneName=deform_groin_l}}
     {targetBone {boneName=deform_groin_r}}
     {targetBone {boneName=def_frontBeltCorr01_r}}
     {targetBone {boneName=def_frontBeltCorr01_l}}
    ]
    constraintWeights [1,1,1,1,1,1,1,1]
   }
   n3 = Look At - Bone: def_beltPinCorr(ComponentPose=@n2, Alpha=0.5) {
    alphaBoolBlend {blendOption=Linear}
    boneToModify {boneName=def_beltPinCorr}
    lookAtLocation=<0.4129 -2.5613 -4.0929>
    lookAtTarget {boneReference {boneName=def_frontBeltCorr01_l}}
    lookAt_Axis {axis {y=-1}}
    lookUp_Axis {axis {x=0}}
   }
   n4 = Look At - Bone: def_beltPinCorr(ComponentPose=@n3, Alpha=0.5) {
    alphaBoolBlend {blendOption=Linear}
    boneToModify {boneName=def_beltPinCorr}
    lookAtLocation=<-0.725 -2.5628 2.351>
    lookAtTarget {boneReference {boneName=def_frontBeltCorr01_r}}
    lookUp_Axis {axis {x=0}}
   }
   n5 = RigidBody!(ComponentPose=@n4, Alpha, OverrideWorldGravity:Vector="0,0,0", ExternalForce) -> (Pose) {
    bClampLinearTranslationLimitToRefPose=true
    bOverrideWorldGravity=true
    overridePhysicsAsset=&/F_MED_GnocchiTea_Soot_Physics
    simulationSpace=BaseBoneSpace
    alphaBoolBlend {blendOption=Linear}
    baseBoneRef {boneName=spine_03}
    componentLinearAccScale=<0.6 0.4 0.4>
    componentLinearVelScale=<0.2 0.2 0.2>
    simSpaceSettings {maxAngularAcceleration=800, maxAngularVelocity=3, maxLinearAcceleration=400, maxLinearVelocity=250, velocityScaleZ=0, worldAlpha=1}
   }
   Output Pose(Result=Component To Local(ComponentPose=@n5)) {layerGroup=None}
  ]}
 EventGraph {nodes [
   Event BlueprintUpdateAnimation()
   > n1 = Gravity Override State Function DEPRECATED(CharacterPartAnimInstance=Self-Reference(), GameTimeInSeconds=GetGameTimeInSeconds(), GravityOverrideParams=Get Gravity Override Params())
   > Set Gravity Override(Gravity Override=@n1)
  ]}
]
variables [
 Gravity Override {type=Vector}
 Gravity Override Params {type=GravityOverrideParamsStruct}
]
